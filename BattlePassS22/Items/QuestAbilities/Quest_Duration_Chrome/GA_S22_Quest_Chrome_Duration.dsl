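; Season 22 Battle Pass quest gameplay ability that reports how long a player stays in the Chrome status.
; On activation it records the server world time, then waits for the ChromePlayer status tag to be removed.
; On removal it sends the elapsed whole seconds as a custom stat event tagged ChromePlayers.Duration to the owning player's quest manager, then ends itself.
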(event Ability|EventActivateAbilityFromEvent
  (params (EventData :type Gameplay%20Event%20Data%20Structure%20%28by%20ref%29))
  ;[1 1 1] Activate ability
  (|CallParentFunction EventData)
  (Variables|Default|SetStartTime (Fortnite|Sync|Time|GetServerWorldTimeSeconds))
  (Ability|Tasks|WaitGameplayTagRemove "(TagName=\"Gameplay.Status.Chrome.ChromePlayer\")"
    (:Removed
      (Quest|SendComplexCustomStatEvent (Variables|GAVars|Data|GetOwningPlayerQuestManager) 0 "(GameplayTags=((TagName=\"Ability.Athena.ChromePlayers.Duration\")))" "(GameplayTags=)" (Math|Float|Truncate (- (Fortnite|Sync|Time|GetServerWorldTimeSeconds) (Variables|Default|GetStartTime))))
      (Ability|EndAbility))))

(event Ability|EventOnEndAbility
  (params (bWasCancelled :type Boolean))
  ;[1 1 1] On End Ability
  (|CallParentFunction bWasCancelled))
